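# Niagara system for the DenimEquip emote that throws a small burst of short-lived, pink-to-red feather meshes.
# Its one emitter, FeathersWorld, spawns a random handful from a squashed sphere with upward-tilted cone velocity and negatively scaled inherited velocity.
# Update reorients the meshes, shrinks them late in life and fades alpha in and out, at half strength outside the front end via User.bIsFrontEnd.
# Continuous spawning and curl noise are present but disabled.
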NiagaraSystem /BRCosmetics/Animation/Game/MainPlayer/Emotes/DenimEquip/FX/NS_Emote_DenimEquip_FeatherBurst
dependencies {
 usedDataInterfaces [&NiagaraDataInterfaceCurve, &NiagaraDataInterfaceVectorCurve, &NiagaraDataInterfaceSkeletalMesh]
 usedDynamicInputs [&MakeVector4_FromLinearColor, &FNTimeOfDayColor]
 usedModules [
  &self:SystemSpawnScript.SystemScriptSource.SystemScriptGraph.NiagaraNodeAssignment_0.SetVariables_#1
  &SystemState
 ]
}
emitters {FeathersWorld {
  emitterUpdateScript [
   &EmitterState()
   &SpawnBurst_Instantaneous(
    Spawn Count=&RandomRangeInt(Minimum=3, Maximum=7)
    Use Spawn Probability=false
    UseLoopCountLimit=false
   )
   &SpawnPerUnit(
    Module.Use Max Movement Threshold=true
    Module.Use Movement Tolerance=true
    Movement Tolerance=1
    Spawn Per Unit=25
    Use Spawn Probability=false
   ) {enabled=false}
   &SpawnRate(
    SpawnRate=&FloatFromCurve(FloatCurve=DataInterface() {propertyValues {Curve {keys [<at 0 1>,<at 0.25 0>]}}}, CurveIndex=$Emitter.NormalizedLoopAge, Scale Curve=100)
    Use Spawn Probability=false
   ) {enabled=false}
  ]
  particleSpawnScript [
   &InitializeParticle(
    Color Maximum=<rgba 0.7181 0.067 0.1037 1>
    Color Minimum=<rgba 1 0.1652 0.2123 1>
    Color Mode=::NewEnumerator2
    Lifetime Max=0.6
    Lifetime Min=0.3
    Lifetime Mode=::NewEnumerator1
    Mesh Scale Max=<0.8 0.35 1.05>
    Mesh Scale Min=<0.6 0.25 0.6>
    Mesh Scale Mode=::NewEnumerator2
    Sprite Rotation Mode=::NewEnumerator2
    UsePositionOffset=false
    Write Material Random=true
   ) {module=InitializeParticle001}
   &InitialMeshOrientation(Use Rotation Vector=false)
   &ApplyInitialForces()
   &AddVelocity(Velocity=&MakeCustomVectorFromBool(
     Bool=$User.bIsFrontEnd
     True Vector=<0 -25 50>
     False Vector=&RandomRangeVector(Minimum=<0 -100 -50>, Maximum=<0 -25 50>)
    )) {enabled=false}
   &InheritVelocity(
    Module.Limit Speed=false
    Velocity Limit=30
    Velocity Scale=&VectorFromCurve(
     VectorCurve=DataInterface() {propertyValues {XCurve {keys [<at 0.5 1>,<at 1 0>]}, YCurve {keys [<at 0.5 1>,<at 1 0>]}, ZCurve {keys [<at 0.5 1>,<at 1 0>]}}}
     Scale Curve=<-1 -1 -1>
    )
   )
   &ShapeLocation(
    MaskSpawnGroup=false
    Non Uniform Scale=<1 1 0.25>
    Offset Coordinate Space=::Local
    Offset Mode=::NewEnumerator0
    Override Seed=false
    Sphere Radius=3
    UseOwnerScale=true
   )
   &AddVelocity(
    Cone Axis=<0 -0.5 1>
    Module.UseConeFalloff=false
    Module.UseVelDistribution=false
    Override Seed=false
    Velocity Mode=::NewEnumerator2
    Velocity Speed=&RandomRangeFloat(Minimum=100, Maximum=150)
   ) {module=AddVelocity001}
  ]
  particleUpdateScript [
   &ParticleState()
   &UpdateMeshOrientation(Rotation Vector=&RandomRangeVector(Minimum=<0 0 0>, Maximum=<1 1 1>))
   &CurlNoiseForce(Bias Noise=false, Noise Frequency=150, Noise Strength=50, Pan Noise=false) {enabled=false}
   &SolveForcesAndVelocity(Clamp Velocity=false, Limit Acceleration=false)
   &CameraOffset(Camera Offset Amount=5)
   &ScaleColor(
    Scale Alpha=&FloatFromCurve(
     FloatCurve=DataInterface() {propertyValues {Curve {keys [
         {interpMode=RCIM_Cubic, tangentMode=RCTM_User, time=0, value=0}
         {interpMode=RCIM_Cubic, tangentMode=RCTM_User, time=0.5, value=1}
         {tangentMode=RCTM_User, time=1, value=0}
        ]}}}
     Scale Curve=&MakeCustomFloatFromBool(Bool=$User.bIsFrontEnd, True Float=1, False Float=0.5)
    )
    ScaleA=true
    ScaleRGB=true
   ) {module=ScaleColor001}
   &ScaleMeshSize(Scale Factor=&VectorFromCurve(VectorCurve=DataInterface() {propertyValues {XCurve {keys [<at 0.6 1>,<at 1 0>]}, YCurve {keys [<at 0.6 1>,<at 1 0>]}, ZCurve {keys [<at 0.6 1>,<at 1 0>]}}}))
  ]
  rendererClasses [&NiagaraMeshRendererProperties]
  renderers [{rendererClass=&NiagaraMeshRendererProperties, rendererIndex=0}]
 }}
summary {
 systemName=NS_Emote_DenimEquip_FeatherBurst
 emitters [{emitterName=FeathersWorld, rendererClasses [&NiagaraMeshRendererProperties]}]
 userVariables [
  User.bIsFrontEnd {type=NiagaraBool}
  User.bIsFrontEndPreview {type=NiagaraBool}
  User.bIsInWater {type=NiagaraBool}
  User.JetActiveCurve {type=NiagaraFloat, defaultValue=0, description=""}
 ]
}
