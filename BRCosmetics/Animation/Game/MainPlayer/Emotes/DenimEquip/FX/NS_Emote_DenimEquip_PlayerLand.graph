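# Niagara system for the landing effect of the DenimEquip player emote, built from timed one-shot bursts.
# A smoke sprite puff drifts out from a shape location, while a slightly later burst of red-pink feather meshes tumbles in curl noise.
# Three flat, upward-facing smoke rings expand at the system origin.
# Smoke opacity and ring tint vary with time of day, and front-end switches adjust puff and feather fades.
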NiagaraSystem /BRCosmetics/Animation/Game/MainPlayer/Emotes/DenimEquip/FX/NS_Emote_DenimEquip_PlayerLand
dependencies {
 usedDataInterfaces [&NiagaraDataInterfaceVectorCurve, &NiagaraDataInterfaceCurve, &NiagaraDataInterfaceSkeletalMesh]
 usedDynamicInputs [
  &MakeVector4_FromLinearColor
  &FNTimeOfDayColor
  &SimulationPosition
 ]
 usedModules [
  &self:SystemSpawnScript.SystemScriptSource.SystemScriptGraph.NiagaraNodeAssignment_0.SetVariables_#1
  &SystemState
 ]
}
emitters {
 Default_Emote_TraversalSmoke_Loop001 {
  emitterSpawnScript [/SetVariables(Emitter.FakeLightActive=true) {isSetParameters=true}]
  emitterUpdateScript [
   &EmitterState()
   &SpawnBurst_Instantaneous(Spawn Count=50, Spawn Time=0.23, Use Spawn Probability=false, UseLoopCountLimit=false)
   &SpawnRate(
    SpawnRate=&FloatFromCurve(FloatCurve=DataInterface() {propertyValues {Curve {keys [<at 0 1>,<at 1 0>]}}}, CurveIndex=$Emitter.NormalizedLoopAge, Scale Curve=335)
    Use Spawn Probability=false
   ) {enabled=false}
   &SpawnPerUnit(
    Module.Use Max Movement Threshold=true
    Module.Use Movement Tolerance=true
    Movement Tolerance=2
    Spawn Spacing=3
    Use Spawn Probability=false
   ) {enabled=false, module=SpawnPerUnit001}
  ]
  particleSpawnScript [
   &PartitionParticles(Number Of Partitions=4)
   &InitializeParticle(
    Color Channel Mode:ENiagaraChannelCorrelation=::NewEnumerator0
    Color Maximum=<rgba 0.35 0.3477 0.336 0.4>
    Color Minimum=<rgba 0.35 0.3404 0.3251 0.3>
    Color Mode=::NewEnumerator2
    Lifetime Max=0.9
    Lifetime Min=0.3
    Lifetime Mode=::NewEnumerator1
    Sprite Rotation Mode=::NewEnumerator2
    Sprite Size Mode=::NewEnumerator1
    Uniform Sprite Size Max=30
    Uniform Sprite Size Min=15
    UsePositionOffset=false
    Write Material Random=true
   ) {module=InitializeParticle001}
   &SkeletalMeshLocation(Bone / Socket Index=5, Bone Sampling Mode=::NewEnumerator4, MaskSpawnGroup=false, Skeletal Mesh=$System.Skeletal Mesh)
   &ShapeLocation(
    Handle Radius=1
    Large Radius=35
    MaskSpawnGroup=false
    Offset Coordinate Space=::Local
    Offset Mode=::NewEnumerator0
    Override Seed=false
    Shape Primitive=::NewEnumerator3
    UseOwnerScale=true
   )
   /SetVariables(Particles.Position=&MakeCustomVectorFromBool(
     Bool=&SetBoolByFloatComparison(A=$User.JetActiveCurve, B=0.1)
     True Vector=$Particles.ShapeLocation.ShapePosition
     False Vector=&AddVectorToPosition(
      Position=$Particles.SkeletalMeshLocation.SampledPosition
      Vector=&RotateVectorByQuaternion(Vector To Rotate=<0 0 -10>, Quaternion=$Engine.Owner.Rotation)
     )
    )) {enabled=false, isSetParameters=true}
   &InheritVelocity(Inherited Velocity Speed Limit=1000, Speed Limit=true, UseSpeedThreshold=false) {enabled=false, module=InheritVelocity001}
   &AddVelocity(Velocity=<0 -2 1>, Velocity Speed Scale=&RandomRangeFloat(Minimum=0, Maximum=100)) {enabled=false, module=AddVelocity001}
   &AddVelocity(Module.Use Falloff Distance=false, Override Seed=false, Velocity Mode=::NewEnumerator1, Velocity Speed=50)
   &AccelerationForce(Acceleration=&UniformRangedVector(Minimum=<-5 -25 -5>, Maximum=<5 -5 5>), Coordinate Space=::Local)
   &ApplyInitialForces()
  ]
  particleUpdateScript [
   &ParticleState()
   &Drag(Drag=10, Use Linear Drag=true, Use Rotational Drag=true) {enabled=false}
   &SolveForcesAndVelocity(Clamp Velocity=false, Limit Acceleration=false)
   &ScaleSpriteSize(
    Uniform Curve Scale=&Lerp_Float(A=1.33, B=2, Alpha=$User.JetActiveCurve)
    Uniform Curve Sprite Scale=DataInterface() {propertyValues {Curve {keys [{arriveTangent=1.491, interpMode=RCIM_Cubic, leaveTangent=1.491, tangentMode=RCTM_User, time=0, value=0.5}, <at 1 3 RCIM_Cubic>]}}}
   )
   &DynamicMaterialParameters(
    Index 0 Param 1=&Add_Float(
     A=&RandomRangeFloat(Minimum=0, Maximum=1)
     B=&FloatFromCurve(
      FloatCurve=DataInterface() {propertyValues {Curve {keys [
          <at 0 0 RCIM_Cubic>
          {arriveTangent=1.5947, interpMode=RCIM_Cubic, leaveTangent=1.5947, tangentMode=RCTM_User, time=0.2557, value=0.5257}
          <at 1 1 RCIM_Cubic>
         ]}}}
      CurveIndex=$Particles.Age
      Scale Curve=0.25
     )
    )
    Index 3 Param 1=&Lerp_Float(A=0, B=1, Alpha=$System.Timeline.Float)
    Index 3 Param 2=150
    Module.Param0WriteEnabled=true
    Module.Param0WriteEnabled1=true
    Module.Param0WriteEnabled2=true
    Module.Param0WriteEnabled3=true
    Module.Param1WriteEnabled=true
    Module.Param1WriteEnabled1=true
    Module.Param1WriteEnabled2=true
    Module.Param1WriteEnabled3=true
    Module.Param2WriteEnabled=true
    Module.Param2WriteEnabled1=true
    Module.Param2WriteEnabled2=true
    Module.Param2WriteEnabled3=true
    Module.Param3WriteEnabled=true
    Module.Param3WriteEnabled1=true
    Module.Param3WriteEnabled2=true
    Module.Param3WriteEnabled3=true
    Write Parameter Index 1=true
    Write Parameter Index 2=true
    Write Parameter Index 3=true
   ) {module=DynamicMaterialParameters001}
   &CameraOffset(Camera Offset Amount=10)
   &ScaleColor(
    Scale Alpha=&Multiply_Float(
     A=&FloatFromCurve(
      FloatCurve=DataInterface() {propertyValues {Curve {keys [{interpMode=RCIM_Cubic, tangentMode=RCTM_User, time=0, value=0}, <at 0.15 1 RCIM_Cubic>, <at 0.9 0>]}}}
      Scale Curve=&FN_FrontEndFloatModifier(Gameplay=0.8, Front End=1)
     )
     B=&FNTimeOfDayScalar(Night Scale=0.75, Morning Scale=0.9, Day Scale=0.9, Evening Scale=0.75)
    )
    ScaleA=true
    ScaleRGB=false
   )
  ]
  rendererClasses [&NiagaraSpriteRendererProperties]
  renderers [
   {rendererClass=&NiagaraSpriteRendererProperties}
   {rendererClass=&NiagaraSpriteRendererProperties, rendererIndex=1}
  ]
 }
 FeathersWorld {
  emitterUpdateScript [
   &EmitterState()
   &SpawnBurst_Instantaneous(Spawn Count=30, Spawn Time=0.4, Use Spawn Probability=false, UseLoopCountLimit=false)
   &SpawnPerUnit(
    Module.Use Max Movement Threshold=true
    Module.Use Movement Tolerance=true
    Movement Tolerance=1
    Spawn Per Unit=25
    Use Spawn Probability=false
   ) {enabled=false}
   &SpawnRate(SpawnRate=5, Use Spawn Probability=false) {enabled=false}
  ]
  particleSpawnScript [
   &InitializeParticle(
    Color Maximum=<rgba 0.7181 0.067 0.1037 1>
    Color Minimum=<rgba 1 0.1652 0.2123 1>
    Color Mode=::NewEnumerator2
    Lifetime Max=0.6
    Lifetime Min=0.3
    Lifetime Mode=::NewEnumerator1
    Mesh Scale Max=<0.8 0.35 1.05>
    Mesh Scale Min=<0.6 0.25 0.6>
    Mesh Scale Mode=::NewEnumerator2
    Sprite Rotation Mode=::NewEnumerator2
    UsePositionOffset=false
    Write Material Random=true
   ) {module=InitializeParticle001}
   &InitialMeshOrientation(Use Rotation Vector=false)
   &ApplyInitialForces()
   &AddVelocity(Velocity=&MakeCustomVectorFromBool(
     Bool=$User.bIsFrontEnd
     True Vector=<0 -25 50>
     False Vector=&RandomRangeVector(Minimum=<0 -100 -50>, Maximum=<0 -25 50>)
    )) {enabled=false}
   &InheritVelocity(
    Module.Limit Speed=false
    Velocity Limit=30
    Velocity Scale=&VectorFromCurve(
     VectorCurve=DataInterface() {propertyValues {XCurve {keys [<at 0.5 1>,<at 1 0>]}, YCurve {keys [<at 0.5 1>,<at 1 0>]}, ZCurve {keys [<at 0.5 1>,<at 1 0>]}}}
     Scale Curve=<-1 -1 -1>
    )
   )
   &ShapeLocation(
    MaskSpawnGroup=false
    Non Uniform Scale=<1 1 0.25>
    Offset=<0 15 25>
    Offset Coordinate Space=::Local
    Offset Mode=::NewEnumerator0
    Override Seed=false
    Sphere Radius=35
    UseOwnerScale=true
   )
   &AddVelocity(
    Module.Use Falloff Distance=false
    Override Seed=false
    Velocity Mode=::NewEnumerator1
    Velocity Speed=&RandomRangeFloat(Minimum=25, Maximum=100)
   ) {module=AddVelocity001}
  ]
  particleUpdateScript [
   &ParticleState()
   &UpdateMeshOrientation(Rotation Vector=&RandomRangeVector(Minimum=<0 0 0>, Maximum=<1 1 1>))
   &CurlNoiseForce(Bias Noise=false, Noise Frequency=150, Noise Strength=50, Pan Noise=false)
   &SolveForcesAndVelocity(Clamp Velocity=false, Limit Acceleration=false)
   &CameraOffset(Camera Offset Amount=5)
   &ScaleColor(
    Scale Alpha=&FloatFromCurve(
     FloatCurve=DataInterface() {propertyValues {Curve {keys [
         {interpMode=RCIM_Cubic, tangentMode=RCTM_User, time=0, value=0}
         {interpMode=RCIM_Cubic, tangentMode=RCTM_User, time=0.5, value=1}
         {tangentMode=RCTM_User, time=1, value=0}
        ]}}}
     Scale Curve=&MakeCustomFloatFromBool(Bool=$User.bIsFrontEnd, True Float=1, False Float=0.5)
    )
    ScaleA=true
    ScaleRGB=true
   ) {module=ScaleColor001}
   &ScaleMeshSize(Scale Factor=&VectorFromCurve(VectorCurve=DataInterface() {propertyValues {XCurve {keys [<at 0.6 1>,<at 1 0>]}, YCurve {keys [<at 0.6 1>,<at 1 0>]}, ZCurve {keys [<at 0.6 1>,<at 1 0>]}}}))
  ]
  rendererClasses [&NiagaraMeshRendererProperties]
  renderers [{rendererClass=&NiagaraMeshRendererProperties, rendererIndex=0}]
 }
 SmokeGroundRings {
  emitterUpdateScript [
   &EmitterState()
   &SpawnBurst_Instantaneous(Spawn Count=3, Use Spawn Probability=false, UseLoopCountLimit=false)
   &SpawnRate(Use Spawn Probability=false) {enabled=false}
  ]
  particleSpawnScript [
   &InitializeParticle(
    Color Mode=::NewEnumerator0
    Mass Mode=::NewEnumerator1
    Sprite Rotation Angle=&UniformRangedFloat(Minimum=-180, Maximum=180)
    Sprite Rotation Mode=::NewEnumerator0
    Sprite Size=&Vector2DFromFloat(Value=300)
    Sprite Size Mode=::NewEnumerator4
    UsePositionOffset=false
    Write Material Random=true
   ) {module=InitializeParticle001}
   /SetVariables(Particles.SpriteFacing=<0 0 1>) {isSetParameters=true}
   &SystemLocation(MaskSpawnGroup=false, Offset=<0 0 0.5>)
  ]
  particleUpdateScript [
   &ParticleState()
   &DynamicMaterialParameters(
    Index 0 Param 1=&FloatFromCurve(FloatCurve=DataInterface() {propertyValues {Curve {keys [<at 0 0.5>,<at 1 2>]}}})
    Module.Param0WriteEnabled=true
    Module.Param1WriteEnabled=false
    Module.Param2WriteEnabled=false
    Module.Param3WriteEnabled=false
   )
   &ScaleSpriteSize(Scale Factor=&Vector2DFromFloat(Value=&FloatFromCurve(FloatCurve=DataInterface() {propertyValues {Curve {keys [
          {arriveTangent=35.6763, interpMode=RCIM_Cubic, leaveTangent=3.7835, tangentMode=RCTM_User, time=0, value=0}
          {interpMode=RCIM_Cubic, leaveTangent=0.6903, time=1.0011, value=1}
         ]}}})))
   &CameraOffset(Camera Offset Amount=20)
   &ScaleColor(
    Scale Alpha=&Multiply_Float(
     A=&FloatFromCurve(
      FloatCurve=DataInterface() {propertyValues {Curve {keys [
          {arriveTangent=0.0112, interpMode=RCIM_Cubic, leaveTangent=0.0112, tangentMode=RCTM_User, time=0, value=0}
          {arriveTangent=0.002, interpMode=RCIM_Cubic, leaveTangent=0.002, time=0.4, value=0.4}
          {interpMode=RCIM_Cubic, leaveTangent=-4.9981, time=1.0011, value=0.002}
         ]}}}
      Scale Curve=&RandomRangeFloat(Minimum=1, Maximum=2)
     )
     B=&FNTimeOfDayScalar(Night Scale=1.5, Morning Scale=2, Day Scale=2, Evening Scale=1.5)
    )
    Scale RGB=&MakeVectorFromLinearColorRGB(Color=&FN_Select_Color_from_Time_Of_Day(
      Morning=<rgba 0.3266 0.3266 0.3266 1.5>
      Day=<rgba 0.3266 0.3266 0.3266 1.5>
      Evening=<rgba 0.1553 0.1761 0.3266 1.5>
      Night=<rgba 0.1871 0.1979 0.3266 1.5>
     ))
    ScaleA=true
    ScaleRGB=true
   )
   &InheritSourceMovement(
    Applied Position Scale=&VectorFromCurve(VectorCurve=DataInterface() {propertyValues {XCurve {keys [<at 0.5 1>,<at 1 0>]}, YCurve {keys [<at 0.5 1>,<at 1 0>]}, ZCurve {keys [<at 0.5 1>,<at 1 0>]}}})
    ApplyForce=false
    ApplyPosition=true
    ApplyPositionFalloff=false
   )
   &SolveForcesAndVelocity(Clamp Velocity=false, Limit Acceleration=false)
  ]
  rendererClasses [&NiagaraSpriteRendererProperties]
  renderers [{rendererClass=&NiagaraSpriteRendererProperties}]
 }
}
summary {
 systemName=NS_Emote_DenimEquip_PlayerLand
 emitters [
  {emitterName=FeathersWorld, rendererClasses [&NiagaraMeshRendererProperties]}
  {emitterName=Default_Emote_TraversalSmoke_Loop001, rendererClasses [&NiagaraSpriteRendererProperties]}
  {emitterName=SmokeGroundRings, rendererClasses [&NiagaraSpriteRendererProperties]}
 ]
 userVariables [
  User.bIsFrontEnd {type=NiagaraBool}
  User.bIsFrontEndPreview {type=NiagaraBool}
  User.bIsInWater {type=NiagaraBool}
  User.JetActiveCurve {type=NiagaraFloat, defaultValue=0, description=""}
 ]
}
